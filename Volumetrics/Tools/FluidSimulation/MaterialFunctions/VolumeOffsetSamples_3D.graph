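# Material function for volume fluid simulation tools that samples a flipbook-style volume texture at the six axis neighbors (±X, ±Y, ±Z) of each voxel.
# EncodeVolumeCoordinates derives a 3D position from XYFrames, offset by one over the X component of VolumeResolution.
# VolumeTextureFunction_NoZBlend samples each position using a separate SampleXYFrames layout, with a static switch clamping to half-voxel edges.
# An extra TopMask output comes from a chain with unconnected inputs, apparently unfinished.
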MaterialFunction /Volumetrics/Tools/FluidSimulation/MaterialFunctions/VolumeOffsetSamples_3D
graph {
 properties {bExposeToLibrary=true}
 nodes [
  Constant2Vector() {r=1}
  n1 = FunctionInput(Preview=TextureObject() {texture=null}) {bUsePreviewValueAsDefault=true, inputName=Texture Object, inputType=FunctionInput_Texture2D, previewValue=<4096 4096 0 1>}
  n2 = FunctionInput() {bUsePreviewValueAsDefault=true, inputName=XYFrames, inputType=FunctionInput_Vector2, sortPriority=2, previewValue=<16 16 0 1>}
  n3 = MaterialFunctionCall!(UVs, XY Frames=@n2) {materialFunction=/Volumetrics/Content/VolumeTextures/MaterialFunctions/EncodeVolumeCoordinates}
  n4 = FunctionInput() {inputName=VolumeResolution, inputType=FunctionInput_Vector2, sortPriority=1, previewValue=<4096 4096 0 1>}
  n5 = Divide(B=ComponentMask(None=@n4) {r=true})
  n6 = Multiply(A=@n5, B=Constant3Vector() {constant=<rgba 1 0 0 0>})
  n7 = Subtract(A=@n3.Result, B=@n6)
  n8 = Multiply(A=@n5) {constB=0.5}
  n9 = OneMinus(None=@n8)
  n10 = Clamp(None=@n7, Min=@n8, Max=@n9)
  n11 = FunctionInput(Preview=StaticBool()) {bUsePreviewValueAsDefault=true, inputName=Clamp Edges, inputType=FunctionInput_StaticBool, sortPriority=10, previewValue=<4096 4096 0 1>}
  n12 = FunctionInput() {bUsePreviewValueAsDefault=true, inputName=SampleXYFrames, inputType=FunctionInput_Vector2, sortPriority=3, previewValue=<16 16 0 1>}
  FunctionOutput(None=MaterialFunctionCall!(Texture Object=@n1, Position=StaticSwitch(True=@n10, False=@n7, Value=@n11), Scale=Constant() {r=1}, Num Frames=@n12) {materialFunction=/Volumetrics/Content/VolumeTextures/MaterialFunctions/VolumeTextureFunction_NoZBlend}.v4) {outputName=-X}
  n13 = Add(A=@n3.Result, B=@n6)
  n14 = Clamp(None=@n13, Min=@n8, Max=@n9)
  FunctionOutput(None=MaterialFunctionCall(Texture Object=@n1, Position=StaticSwitch(True=@n14, False=@n13, Value=@n11), Scale=Constant() {r=1}, Num Frames=@n12) {materialFunction=&VolumeTextureFunction_NoZBlend}.v4) {outputName=+X, sortPriority=1}
  n15 = Ceil()
  n16 = MaterialFunctionCall(Float2=Floor(None=Divide(A=@n15))) {materialFunction=&BreakOutFloat2Components}
  n17 = Min(A=@n16.R, B=@n16.R)
  FunctionOutput(None=OneMinus(None=Saturate(None=@n17))) {outputName=TopMask, sortPriority=10}
  n18 = MaterialFunctionCall(XY Frames=@n2) {materialFunction=&EncodeVolumeCoordinates}
  n19 = Divide(B=ComponentMask(None=@n4) {r=true})
  n20 = Multiply(A=@n19, B=Constant3Vector() {constant=<rgba 0 1 0 0>})
  n21 = Subtract(A=@n18.Result, B=@n20)
  n22 = Multiply(A=@n19) {constB=0.5}
  n23 = OneMinus(None=@n22)
  n24 = Clamp(None=@n21, Min=@n22, Max=@n23)
  FunctionOutput(None=MaterialFunctionCall(Texture Object=@n1, Position=StaticSwitch(True=@n24, False=@n21, Value=@n11), Scale=Constant() {r=1}, Num Frames=@n12) {materialFunction=&VolumeTextureFunction_NoZBlend}.v4) {outputName=-Y, sortPriority=2}
  n25 = Add(A=@n18.Result, B=@n20)
  n26 = Clamp(None=@n25, Min=@n22, Max=@n23)
  FunctionOutput(None=MaterialFunctionCall(Texture Object=@n1, Position=StaticSwitch(True=@n26, False=@n25, Value=@n11), Scale=Constant() {r=1}, Num Frames=@n12) {materialFunction=&VolumeTextureFunction_NoZBlend}.v4) {outputName=+Y, sortPriority=3}
  n27 = MaterialFunctionCall(XY Frames=@n2) {materialFunction=&EncodeVolumeCoordinates}
  n28 = Divide(B=ComponentMask(None=@n4) {r=true})
  n29 = Multiply(A=@n28, B=Constant3Vector() {constant=<rgba 0 0 1 0>})
  n30 = Subtract(A=@n27.Result, B=@n29)
  n31 = Multiply(A=@n28) {constB=0.5}
  n32 = OneMinus(None=@n31)
  n33 = Clamp(None=@n30, Min=@n31, Max=@n32)
  FunctionOutput(None=MaterialFunctionCall(Texture Object=@n1, Position=StaticSwitch(True=@n33, False=@n30, Value=@n11), Scale=Constant() {r=1}, Num Frames=@n12) {materialFunction=&VolumeTextureFunction_NoZBlend}.v4) {outputName=-Z, sortPriority=4}
  n34 = Add(A=@n27.Result, B=@n29)
  n35 = Clamp(None=@n34, Min=@n31, Max=@n32)
  FunctionOutput(None=MaterialFunctionCall(Texture Object=@n1, Position=StaticSwitch(True=@n35, False=@n34, Value=@n11), Scale=Constant() {r=1}, Num Frames=@n12) {materialFunction=&VolumeTextureFunction_NoZBlend}.v4) {outputName=+Z, sortPriority=5}
  n36 = MaterialFunctionCall() {materialFunction=&BreakOutFloat2Components}
  Saturate(None=Max(A=@n36.R, B=@n36.R))
  TextureCoordinate()
 ]
}
