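# VolumeCoordOffset is a material function in the Volumetrics volume texture set that shifts a UV by a texel-scaled offset.
# It takes the reciprocal of the RG channels of the TextureResolution vector parameter to get texel size, multiplies it by the OffsetDir input, and adds that to the UV input.
# The UV input falls back to the texture coordinate, and OffsetDir to a one-texel step along U.
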MaterialFunction /Volumetrics/Content/VolumeTextures/MaterialFunctions/VolumeCoordOffset
graph {nodes [
  n1 = TextureCoordinate()
  n2 = Divide(B=ComponentMask(None=VectorParameter() {parameterName=TextureResolution, defaultValue=<rgba 4096 4096 0 0>}.RGB) {g=true, r=true})
  n3 = FunctionInput() {bUsePreviewValueAsDefault=true, inputName=OffsetDir, inputType=FunctionInput_Vector2, sortPriority=1, previewValue=<1 0 0 1>}
  FunctionOutput(None=Add(A=FunctionInput(Preview=@n1) {bUsePreviewValueAsDefault=true, inputName=UV, inputType=FunctionInput_Vector2}, B=Multiply(A=@n2, B=@n3))) {outputName=UV}
 ]}
