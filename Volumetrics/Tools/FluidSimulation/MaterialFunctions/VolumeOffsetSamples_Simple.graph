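# Material function for fluid simulation that samples the six axis neighbors of a voxel in a volume texture stored as a 2D grid of slices.
# X and Y neighbors use VolumeCoordOffset, with optional VolumeCoordClamp through the Clamp Edges static switch.
# Z neighbors convert the tile to a 1D slice index, step it by one and convert back to a UV shift.
# It also outputs a TopMask derived from the XYFrames tile coordinates.
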MaterialFunction /Volumetrics/Tools/FluidSimulation/MaterialFunctions/VolumeOffsetSamples_Simple
graph {
 properties {bExposeToLibrary=true}
 nodes [
  n1 = ComponentMask(None=VectorParameter() {parameterName=XYFrames, defaultValue=<rgba 16 16 0 0>}.RGB) {g=true, r=true}
  n2 = Multiply(A=TextureCoordinate(), B=@n1)
  n3 = MaterialFunctionCall(Float2=Floor(None=@n2)) {materialFunction=&BreakOutFloat2Components}
  n4 = Max(A=@n3.R, B=@n3.R)
  n5 = Multiply(A=Saturate(None=OneMinus(None=@n4)))
  n6 = Ceil(None=@n2)
  n7 = MaterialFunctionCall(Float2=Floor(None=Divide(A=@n6, B=@n1))) {materialFunction=&BreakOutFloat2Components}
  n8 = Saturate(None=Min(A=@n7.R, B=@n7.R))
  n9 = Multiply(A=@n8) {constB=-1}
  n10 = VectorParameter() {parameterName=XYFrames, defaultValue=<rgba 16 16 0 0>}
  Add(A=TextureCoordinate(), B=Multiply(A=Add(A=@n5, B=@n9), B=Divide(B=@n10.R)))
  n11 = Constant2Vector() {r=-1}
  n12 = FunctionInput(Preview=StaticBool()) {bUsePreviewValueAsDefault=true, inputName=Clamp Edges, inputType=FunctionInput_StaticBool, sortPriority=10, previewValue=<4096 4096 0 1>}
  n13 = StaticSwitch(
   True=MaterialFunctionCall() {materialFunction=/Volumetrics/Content/VolumeTextures/MaterialFunctions/VolumeCoordClamp}.UV
   False=TextureCoordinate()
   Value=@n12
  )
  n14 = FunctionInput(Preview=TextureObject() {texture=/Volumetrics/Tools/FluidSimulation/Textures/RT_Tiny}) {bUsePreviewValueAsDefault=true, inputName=Texture Object, inputType=FunctionInput_Texture2D, previewValue=<4096 4096 0 1>}
  FunctionOutput(None=TextureSample(
    UVs=MaterialFunctionCall!(UV=@n13, OffsetDir=@n11) {materialFunction=/Volumetrics/Content/VolumeTextures/MaterialFunctions/VolumeCoordOffset}.UV
    Tex=@n14
   ).RGB) {outputName=-X}
  n15 = Constant2Vector() {r=1}
  FunctionOutput(None=TextureSample(
    UVs=MaterialFunctionCall(UV=@n13, OffsetDir=@n15) {materialFunction=&VolumeCoordOffset}.UV
    Tex=@n14
   ).RGB) {outputName=+X, sortPriority=1}
  FunctionOutput(None=@n8) {outputName=TopMask, sortPriority=10}
  n16 = Constant2Vector() {g=-1}
  n17 = StaticSwitch(
   True=MaterialFunctionCall() {materialFunction=&VolumeCoordClamp}.UV
   False=TextureCoordinate()
   Value=@n12
  )
  FunctionOutput(None=TextureSample(
    UVs=MaterialFunctionCall(UV=@n17, OffsetDir=@n16) {materialFunction=&VolumeCoordOffset}.UV
    Tex=@n14
   ).RGB) {outputName=-Y, sortPriority=2}
  n18 = Constant2Vector() {g=1}
  FunctionOutput(None=TextureSample(
    UVs=MaterialFunctionCall(UV=@n17, OffsetDir=@n18) {materialFunction=&VolumeCoordOffset}.UV
    Tex=@n14
   ).RGB) {outputName=+Y, sortPriority=3}
  n19 = StaticSwitch(
   True=MaterialFunctionCall() {materialFunction=&VolumeCoordClamp}.UV
   False=TextureCoordinate()
   Value=@n12
  )
  n20 = ComponentMask(None=VectorParameter() {parameterName=XYFrames, defaultValue=<rgba 16 16 0 0>}.RGB) {g=true, r=true}
  n21 = Floor(None=Multiply(A=@n19, B=@n20))
  n22 = MaterialFunctionCall!(2D Index=@n21, Columns X=@n20) {materialFunction=/Engine/Functions/Engine_MaterialFunctions01/Coordinates/2Dto1DIndex}
  n23 = Subtract(A=@n22.1D Index)
  n24 = ComponentMask(None=VectorParameter() {parameterName=XYFrames, defaultValue=<rgba 16 16 0 0>}.RGB) {g=true, r=true}
  n25 = ComponentMask(None=VectorParameter() {parameterName=XYFrames, defaultValue=<rgba 16 16 0 0>}.RGB) {g=true, r=true}
  n26 = Divide(
   A=Subtract(
    A=MaterialFunctionCall!(1D Index=@n23, Cells XY=@n24) {materialFunction=/Engine/Functions/Engine_MaterialFunctions01/Coordinates/1Dto2DIndex}.2D Index
    B=@n21
   )
   B=@n25
  )
  FunctionOutput(None=TextureSample(UVs=Add(A=@n19, B=@n26), Tex=@n14).RGB) {outputName=-Z, sortPriority=4}
  n27 = Add(A=@n22.1D Index)
  n28 = Divide(
   A=Subtract(
    A=MaterialFunctionCall(1D Index=@n27, Cells XY=@n24) {materialFunction=&1Dto2DIndex}.2D Index
    B=@n21
   )
   B=@n25
  )
  FunctionOutput(None=TextureSample(UVs=Add(A=@n19, B=@n28), Tex=@n14).RGB) {outputName=+Z, sortPriority=5}
  TextureCoordinate()
 ]
}
