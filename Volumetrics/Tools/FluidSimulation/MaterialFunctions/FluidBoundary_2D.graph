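# Material function for 2D fluid simulation that outputs four 0/1 edge masks: Negative X, Positive X, Negative Y and Positive Y.
# It converts UVs to integer texel coordinates using the Resolution parameter and floor, then compares each axis with If nodes.
# Positive masks cover texels at or beyond Resolution minus BorderTexels.
# Negative masks match only texel zero, so BorderTexels does not affect their width.
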MaterialFunction /Volumetrics/Tools/FluidSimulation/MaterialFunctions/FluidBoundary_2D
graph {
 properties {bExposeToLibrary=true}
 nodes [
  n1 = Floor(None=Multiply(A=TextureCoordinate(), B=ScalarParameter() {defaultValue=256, parameterName=Resolution}))
  n2 = ComponentMask(None=@n1) {g=true}
  n3 = Subtract(A=ScalarParameter() {defaultValue=256, parameterName=Resolution}, B=ScalarParameter() {defaultValue=1, parameterName=BorderTexels})
  n4 = Constant() {r=1}
  FunctionOutput(None=If(A=@n2, B=@n3, "A > B"=@n4, "A == B"=@n4, "A < B"=Constant()) {desc=Top Boundary}) {outputName=Positive Y, sortPriority=3}
  n5 = Constant()
  FunctionOutput(None=If(A=@n2, B=Constant(), "A > B"=@n5, "A == B"=Constant() {r=1}, "A < B"=@n5) {desc=Top Boundary}) {outputName=Negative Y, sortPriority=2}
  n6 = ComponentMask(None=@n1) {r=true}
  n7 = Constant() {r=1}
  FunctionOutput(None=If(A=@n6, B=@n3, "A > B"=@n7, "A == B"=@n7, "A < B"=Constant()) {desc=Top Boundary}) {outputName=Positive X, sortPriority=1}
  n8 = Constant()
  FunctionOutput(None=If(A=@n6, B=Constant(), "A > B"=@n8, "A == B"=Constant() {r=1}, "A < B"=@n8) {desc=Top Boundary}) {outputName=Negative X}
 ]
}
