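# Material function for fluid simulation that outputs six 0/1 masks marking boundary texels of a 3D volume stored as a flipbook of slices.
# UVs scale by XYFrames into tile index and local coordinate.
# Local texel position, floored or ceiled against FrameResolution, yields the X and Y masks.
# The flattened slice index yields Z masks, with the positive limit assuming a square grid (XYFrames.R squared) less BorderTexels.
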MaterialFunction /Volumetrics/Tools/FluidSimulation/MaterialFunctions/FluidBoundary_3D
graph {
 properties {bExposeToLibrary=true}
 nodes [
  n1 = ComponentMask(None=VectorParameter() {parameterName=XYFrames, defaultValue=<rgba 8 8 0 0>}.RGB) {g=true, r=true}
  n2 = Multiply(A=TextureCoordinate(), B=@n1)
  n3 = Multiply(
   A=ComponentMask(None=VectorParameter() {parameterName=FrameResolution, defaultValue=<rgba 256 256 0 0>}.RGB) {g=true, r=true}
   B=ComponentMask(None=Frac(None=@n2)) {g=true, r=true}
  )
  n4 = Ceil(None=@n3)
  n5 = Constant() {r=1}
  FunctionOutput(None=If(
    A=ComponentMask(None=@n4) {g=true}
    B=VectorParameter() {parameterName=FrameResolution, defaultValue=<rgba 8 8 0 0>}.G
    "A > B"=@n5
    "A == B"=@n5
    "A < B"=Constant()
   )) {outputName=Positive Y, sortPriority=4}
  n6 = Floor(None=@n3)
  n7 = Constant()
  FunctionOutput(None=If(A=ComponentMask(None=@n6) {g=true}, B=Constant(), "A > B"=@n7, "A == B"=Constant() {r=1}, "A < B"=@n7)) {outputName=Negative Y, sortPriority=1}
  n8 = Constant() {r=1}
  FunctionOutput(None=If(
    A=ComponentMask(None=@n4) {r=true}
    B=VectorParameter() {parameterName=FrameResolution, defaultValue=<rgba 8 8 0 0>}.R
    "A > B"=@n8
    "A == B"=@n8
    "A < B"=Constant()
   )) {outputName=Positive X, sortPriority=3}
  n9 = Constant()
  FunctionOutput(None=If(A=ComponentMask(None=@n6) {r=true}, B=Constant(), "A > B"=@n9, "A == B"=Constant() {r=1}, "A < B"=@n9)) {outputName=Negative X}
  n10 = MaterialFunctionCall!(2D Index=Floor(None=@n2), Columns X=@n1) {materialFunction=/Engine/Functions/Engine_MaterialFunctions01/Coordinates/2Dto1DIndex}
  n11 = VectorParameter() {parameterName=XYFrames, defaultValue=<rgba 16 16 0 0>}
  n12 = Multiply(A=@n11.R, B=@n11.R)
  n13 = ScalarParameter() {defaultValue=1, parameterName=BorderTexels}
  n14 = Constant() {r=1}
  FunctionOutput(None=If(A=@n10.1D Index, B=Subtract(A=@n12, B=@n13), "A > B"=@n14, "A == B"=@n14, "A < B"=Constant())) {outputName=Positive Z, sortPriority=5}
  n15 = Constant()
  FunctionOutput(None=If(A=@n10.1D Index, B=Constant(), "A > B"=@n15, "A == B"=Constant() {r=1}, "A < B"=@n15)) {outputName=Negative Z, sortPriority=2}
  MaterialFunctionCall(Number=@n10.1D Index, UVs=Frac(None=@n2)) {materialFunction=&DebugScalarValues}
 ]
}
