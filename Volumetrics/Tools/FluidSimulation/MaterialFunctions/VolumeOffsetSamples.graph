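# Material function for fluid simulation that samples the six axis neighbors of a volume texture stored as a 2D grid of slices.
# X and Y samples shift UVs by one texel, while Z samples step the slice index through 2D/1D index conversions.
# A Clamp Edges static switch keeps X/Y samples inside their tile and masks Z offsets at end slices.
# It outputs the six samples plus a TopMask.
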MaterialFunction /Volumetrics/Tools/FluidSimulation/MaterialFunctions/VolumeOffsetSamples
graph {
 properties {bExposeToLibrary=true}
 nodes [
  n1 = FunctionInput() {inputName=Resolution, inputType=FunctionInput_Vector2, sortPriority=1, previewValue=<4096 4096 0 1>}
  n2 = Multiply(A=Divide(B=ComponentMask(None=@n1) {r=true}), B=Constant2Vector() {r=1})
  n3 = Subtract(A=TextureCoordinate(), B=@n2)
  n4 = TextureCoordinate()
  n5 = FunctionInput() {bUsePreviewValueAsDefault=true, inputName=XYFrames, inputType=FunctionInput_Vector2, sortPriority=2, previewValue=<16 16 0 1>}
  n6 = Multiply(A=@n4, B=@n5)
  n7 = Multiply(A=Divide(B=ComponentMask(None=@n1) {r=true}) {desc=VolumeTexelSize}) {constB=0.5}
  n8 = Add(A=Divide(A=Floor(None=@n6), B=@n5), B=@n7)
  n9 = FunctionInput(Preview=StaticBool()) {bUsePreviewValueAsDefault=true, inputName=Clamp Edges, inputType=FunctionInput_StaticBool, sortPriority=10, previewValue=<4096 4096 0 1>}
  n10 = FunctionInput(Preview=TextureObject() {texture=null}) {bUsePreviewValueAsDefault=true, inputName=Texture Object, inputType=FunctionInput_Texture2D, previewValue=<4096 4096 0 1>}
  n11 = TextureSample(UVs=StaticSwitch(True=Max(A=@n3, B=@n8), False=@n3, Value=@n9), Tex=@n10)
  FunctionOutput(None=AppendVector(A=@n11.RGB, B=@n11.RGB)) {outputName=-X}
  n12 = Add(A=@n2, B=@n4)
  n13 = Subtract(A=Divide(A=Ceil(None=@n6), B=@n5), B=@n7)
  n14 = TextureSample(UVs=StaticSwitch(True=Min(A=@n12, B=@n13), False=@n12, Value=@n9), Tex=@n10)
  FunctionOutput(None=AppendVector(A=@n14.RGB, B=@n14.RGB)) {outputName=+X, sortPriority=1}
  n15 = Multiply(A=@n5) {desc=XY Frames}
  n16 = Multiply(A=TextureCoordinate(), B=@n15)
  n17 = Ceil(None=@n16)
  n18 = MaterialFunctionCall(Float2=Floor(None=Divide(A=@n17, B=@n15))) {materialFunction=&BreakOutFloat2Components}
  n19 = OneMinus(None=Saturate(None=Min(A=@n18.R, B=@n18.R)))
  FunctionOutput(None=@n19) {outputName=TopMask, sortPriority=10}
  n20 = Multiply(A=Divide(B=ComponentMask(None=@n1) {g=true}), B=Constant2Vector() {g=1})
  n21 = Subtract(A=TextureCoordinate(), B=@n20)
  n22 = TextureCoordinate()
  n23 = Multiply(A=@n5)
  n24 = Multiply(A=@n22, B=@n23)
  n25 = Multiply(A=Divide(B=ComponentMask(None=@n1) {g=true}) {desc=VolumeTexelSize}) {constB=0.5}
  n26 = Add(A=Divide(A=Floor(None=@n24), B=@n23), B=@n25)
  n27 = TextureSample(UVs=StaticSwitch(True=Max(A=@n21, B=@n26), False=@n21, Value=@n9), Tex=@n10)
  FunctionOutput(None=AppendVector(A=@n27.RGB, B=@n27.RGB)) {outputName=-Y, sortPriority=2}
  n28 = Add(A=@n20, B=@n22)
  n29 = Subtract(A=Divide(A=Ceil(None=@n24), B=@n23), B=@n25)
  n30 = TextureSample(UVs=StaticSwitch(True=Min(A=@n28, B=@n29), False=@n28, Value=@n9), Tex=@n10)
  FunctionOutput(None=AppendVector(A=@n30.RGB, B=@n30.RGB)) {outputName=+Y, sortPriority=3}
  n31 = Floor(None=@n16)
  n32 = MaterialFunctionCall!(2D Index=@n31, Columns X=@n15) {materialFunction=/Engine/Functions/Engine_MaterialFunctions01/Coordinates/2Dto1DIndex}
  n33 = Subtract(A=@n32.1D Index)
  n34 = Divide(
   A=Subtract(
    A=MaterialFunctionCall!(1D Index=@n33, Cells XY=@n15) {materialFunction=/Engine/Functions/Engine_MaterialFunctions01/Coordinates/1Dto2DIndex}.2D Index
    B=@n31
   )
   B=@n15
  )
  n35 = MaterialFunctionCall(Float2=@n31) {materialFunction=&BreakOutFloat2Components}
  n36 = Multiply(A=@n34, B=Saturate(None=Max(A=@n35.R, B=@n35.R)))
  n37 = TextureSample(UVs=Add(A=TextureCoordinate(), B=StaticSwitch(True=@n36, False=@n34, Value=@n9)), Tex=@n10)
  FunctionOutput(None=AppendVector(A=@n37.RGB, B=@n37.RGB)) {outputName=-Z, sortPriority=4}
  n38 = Add(A=@n32.1D Index)
  n39 = Divide(
   A=Subtract(
    A=MaterialFunctionCall(1D Index=@n38, Cells XY=@n15) {materialFunction=&1Dto2DIndex}.2D Index
    B=@n31
   )
   B=@n15
  )
  n40 = Multiply(A=@n39, B=@n19)
  n41 = TextureSample(UVs=Add(A=TextureCoordinate(), B=StaticSwitch(True=@n40, False=@n39, Value=@n9)), Tex=@n10)
  FunctionOutput(None=AppendVector(A=@n41.RGB, B=@n41.RGB)) {outputName=+Z, sortPriority=5}
 ]
}
